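; AudioFadeOverrideComponent is a Blueprint that listens for audio fade change events from player 0's camera manager.
; On BeginPlay it caches that manager and binds to OnAudioFadeChangeEvent, only when the cast of its owning controller to FortPlayerControllerSpectating fails, excluding spectators.
; EndPlay unbinds through the cached reference when it is valid.
; The fade handler, which receives a fade-out flag and time, and Tick are both empty here.
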(event EventTick
  (params (DeltaSeconds :type Float%20%28single-precision%29)))

(event EventBeginPlay
  (bind _returnvalue (Game|GetPlayerCameraManager 0))
  (Utilities|IsValid _returnvalue
    (:Is%20Valid
      (Utilities|Casting|CastToFortPlayerControllerSpectating (Game|Player|GetOwningPlayerController _returnvalue)
        (:CastFailed
          (bind _output_get (Variables|Default|SetCameraManagerRef _returnvalue))
          (EventDispatchers|BindEventtoOnAudioFadeChangeEvent _output_get (delegate Custom|On%20Audio%20Fade%20Change%20Event)))))))

(event Custom|On%20Audio%20Fade%20Change%20Event
  (params (bFadeOut :type Boolean) (FadeTime :type Float%20%28single-precision%29)))

(event EventEndPlay
  (params (EndPlayReason :type EEndPlayReason%20Enum :default "Destroyed"))
  (bind _cameramanagerref (Variables|Default|GetCameraManagerRef)
    (:then
      (EventDispatchers|UnbindEventfromOnAudioFadeChangeEvent _cameramanagerref (delegate Custom|On%20Audio%20Fade%20Change%20Event)))
    (:else)))
